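; Looping gameplay cue that shows concert rain and haze for the Papaya event to the locally controlled player.
; On looping start it moves to a fixed world location, activates the rain particles and fades in; for other pawns it destroys a haze mesh component.
; On removal the local copy deactivates the rain and fades out.
; One RainFade timeline, reversed for fade-in, drives Emissive on a haze dynamic material created in the construction script.
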(event EventOnLoopingStart(Cascade)
  (params (MyTarget :type Actor%20Object%20Reference) (Parameters :type Gameplay%20Cue%20Parameters%20Structure%20%28by%20ref%29) (ParticleComponents :type Array%20of%20Cascade%20Particle%20System%20Component%20Object%20References) (AudioComponents :type Array%20of%20Audio%20Component%20Object%20References))
  (bind _asfort_player_pawn (Utilities|Casting|CastToFortPlayerPawn MyTarget))
  (if (Pawn|IsLocallyControlled _asfort_player_pawn)
    (Transformation|SetActorLocation "-1028.000000,24076.000000,3500.000000")
    (Components|Activation|Activate (Variables|Default|GetP_Papaya_Concert_Rain) true)
    (CallFunction|FadeIn)
    (else
      (Components|DestroyComponent (Variables|Default|GetSM_NoTickHaze) 0))))

(event Custom|FadeOut
  (bind (_direction _fade) (Timeline|RainFade PlayFromStart)
    (:Update
      (Rendering|Material|SetScalarParameterValue (Variables|Default|GetHaze_MID) "Emissive" _fade :!class /Script/Engine.MaterialInstanceDynamic))))

(event Custom|Fade%20In
  (Timeline|RainFade ReverseFromEnd))

(event EventOnRemoval(Cascade)
  (params (MyTarget :type Actor%20Object%20Reference) (Parameters :type Gameplay%20Cue%20Parameters%20Structure%20%28by%20ref%29) (ParticleComponents :type Array%20of%20Cascade%20Particle%20System%20Component%20Object%20References) (AudioComponents :type Array%20of%20Audio%20Component%20Object%20References) (BurstCameraShakeInstance :type Camera%20Shake%20Base%20Object%20Reference) (BurstDecalInstance :type Decal%20Actor%20Object%20Reference))
  (bind _asfort_player_pawn (Utilities|Casting|CastToFortPlayerPawn MyTarget))
  (if (Pawn|IsLocallyControlled _asfort_player_pawn)
    (Components|Activation|Deactivate (Variables|Default|GetP_Papaya_Concert_Rain))
    (CallFunction|FadeOut)))

(fn ConstructionScript
  (bind _returnvalue (Rendering|Material|CreateDynamicMaterialInstance (Variables|Default|GetSM_NoTickHaze) 0 "/CreativeSpecialEvents/FX/Materials/MI_Papaya_NoTick_Haze_ConcertRain.MI_Papaya_NoTick_Haze_ConcertRain" "ConcertRainHaze" :!class /Script/Engine.PrimitiveComponent))
  (bind _output_get (Variables|Default|SetHaze_MID _returnvalue))
  (Rendering|Material|SetMaterial (Variables|Default|GetSM_NoTickHaze1) 0 _output_get))
