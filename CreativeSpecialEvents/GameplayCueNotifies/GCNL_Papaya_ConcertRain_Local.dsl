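; Looping gameplay cue that shows the Papaya concert rain Niagara effect only for the local viewer.
; On loop start it attaches the rain to the replay spectator's root component during replay playback, otherwise to the mesh root of the target pawn if locally controlled.
; The component spawns with manual-release pooling and is stored in Spawned Local Rain, then released back to the pool on cue removal.
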(event EventOnLoopingStart(Cascade)
  (params (MyTarget :type Actor%20Object%20Reference) (Parameters :type Gameplay%20Cue%20Parameters%20Structure%20%28by%20ref%29) (ParticleComponents :type Array%20of%20Cascade%20Particle%20System%20Component%20Object%20References) (AudioComponents :type Array%20of%20Audio%20Component%20Object%20References))
  (if (Replay|GetIsPlayingReplay MyTarget)
    (bind _asfort_replay_spectator (Utilities|Casting|CastToFortReplaySpectator (Fortnite|GetFirstLocalFortPlayerController MyTarget false true)))
    (bind _returnvalue (Effects|SpawnFXSystemAttached :FXSystemAsset "/CreativeSpecialEvents/FX/Niagara/NS_Papaya_ConcertRain.NS_Papaya_ConcertRain" :AttachToComponent (Variables|/Script/Engine.Actor|GetRootComponent _asfort_replay_spectator) :AttachPointName "None" :Scale "1.000000,1.000000,1.000000" :LocationType "SnapToTarget" :bAutoDestroy true :PoolingMethod "ManualRelease" :bAutoActivate true))
    (Variables|Default|SetSpawned%20Local%20Rain _returnvalue)
    (else
      (bind _asfort_player_pawn (Utilities|Casting|CastToFortPlayerPawn MyTarget))
      (if (Pawn|IsLocallyControlled _asfort_player_pawn)
        (bind _returnvalue_2 (Effects|SpawnFXSystemAttached :FXSystemAsset "/CreativeSpecialEvents/FX/Niagara/NS_Papaya_ConcertRain.NS_Papaya_ConcertRain" :AttachToComponent (Variables|/Script/Engine.Character|GetMesh _asfort_player_pawn) :AttachPointName "root" :Scale "1.000000,1.000000,1.000000" :LocationType "SnapToTarget" :bAutoDestroy true :PoolingMethod "ManualRelease" :bAutoActivate true))
        (Variables|Default|SetSpawned%20Local%20Rain _returnvalue_2)))))

(event EventOnRemoval(Cascade)
  (params (MyTarget :type Actor%20Object%20Reference) (Parameters :type Gameplay%20Cue%20Parameters%20Structure%20%28by%20ref%29) (ParticleComponents :type Array%20of%20Cascade%20Particle%20System%20Component%20Object%20References) (AudioComponents :type Array%20of%20Audio%20Component%20Object%20References) (BurstCameraShakeInstance :type Camera%20Shake%20Base%20Object%20Reference) (BurstDecalInstance :type Decal%20Actor%20Object%20Reference))
  (bind _spawned_local_rain (Variables|Default|GetSpawned%20Local%20Rain)
    (:then
      (Effects|Components|ParticleSystem|ReleasetoPool _spawned_local_rain))
    (:else)))

(fn ConstructionScript)
